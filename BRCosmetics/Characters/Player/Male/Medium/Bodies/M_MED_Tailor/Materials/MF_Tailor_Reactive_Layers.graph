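# Material function that builds one reactive layer mask for the Tailor character's material from noise and a gradient texture.
# GradientPath plus LayerAlpha is scaled into a stretched V coordinate that samples the gradient, so LayerAlpha slides the band along the path.
# Noise-modulated red and strongly boosted green are summed, then output raw as LayerColorMask and saturated as Clamped.
# A LayerAlpha-driven lerp is left unconnected.
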MaterialFunction /BRCosmetics/Characters/Player/Male/Medium/Bodies/M_MED_Tailor/Materials/MF_Tailor_Reactive_Layers
graph {nodes [
  n1 = FunctionInput() {bUsePreviewValueAsDefault=true, inputName=Noise, inputType=FunctionInput_Scalar, previewValue=<0 0 0 0>}
  n2 = FunctionInput() {bUsePreviewValueAsDefault=true, inputName=GradientPath, inputType=FunctionInput_Scalar, previewValue=<0 0 0 0>}
  n3 = FunctionInput() {bUsePreviewValueAsDefault=true, inputName=LayerAlpha, inputType=FunctionInput_Scalar, previewValue=<0 0 0 0>}
  n4 = FunctionInput() {bUsePreviewValueAsDefault=true, inputName=Gradient, inputType=FunctionInput_Texture2D, previewValue=<0 0 0 0>}
  n5 = TextureSample(UVs=Multiply(A=Add(A=@n2, B=@n3), B=Constant2Vector() {desc=Make it longer, g=0.5}), Tex=@n4) {samplerSource=SSM_Wrap_WorldGroupSettings, samplerType=SAMPLERTYPE_LinearColor}
  n6 = Add(A=Multiply(A=Constant() {r=2}, B=Multiply(A=@n1, B=@n5.R)), B=Multiply(A=@n5.G, B=Constant() {r=15}) {constB=5})
  FunctionOutput(None=Saturate(None=@n6)) {outputName=Clamped}
  FunctionOutput(None=@n6) {outputName=LayerColorMask}
  LinearInterpolate(A=Constant() {r=1.8}, B=Constant() {r=3.4}, Alpha=@n3)
 ]}
