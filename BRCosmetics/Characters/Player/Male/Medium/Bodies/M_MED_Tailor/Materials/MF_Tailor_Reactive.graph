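# Material function that adds an elimination-reactive gold effect to the Tailor character's incoming material attributes.
# Two sphere masks along Mask-A, driven by Elimination-Start/Exit, form a noise-broken front (noise off on Forward/Mobile) that, within Mask-B, turns surfaces gold and metallic, adjusting roughness and subsurface.
# Emissive adds elimination glow overlaid by three MF_Tailor_Reactive_Layers color layers, limited by Mask-G times Mask-B.
# Several leftover debug and animation chains are unconnected to the output.
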MaterialFunction /BRCosmetics/Characters/Player/Male/Medium/Bodies/M_MED_Tailor/Materials/MF_Tailor_Reactive
graph {nodes [
  n1 = ScalarParameter() {parameterName=Elimination-Start}
  Add(A=@n1, B=ScalarParameter() {defaultValue=0.5, parameterName=DEBUG-Elim})
  Constant3Vector() {constant=<rgba 1 0.4803 0.0355 1>}
  Constant3Vector() {constant=<rgba 0.0824 0.3963 1 1>}
  Constant3Vector() {constant=<rgba 0.037 1 0.7482 1>}
  Constant3Vector() {constant=<rgba 2 0.0341 0 1>}
  Constant3Vector() {constant=<rgba 0.5647 0.3813 0.15 1>}
  Constant() {r=2}
  Constant() {r=-2}
  n2 = GetMaterialAttributes(None=FunctionInput() {inputName=Material, inputType=FunctionInput_MaterialAttributes}) {attributeGetTypes [
    #1
    #2
    #3
    #4
    #5
    #6
   ]}
  n3 = FunctionInput() {bUsePreviewValueAsDefault=true, inputName=Mask-B, inputType=FunctionInput_Scalar, previewValue=<0 0 0 0>}
  n4 = Constant() {desc="Elimination-Rad should use the\r\nnegative of this as the min", r=0.1}
  n5 = LinearInterpolate(A=Multiply(A=Constant() {r=-1}, B=@n4), B=Constant() {desc="0 to 1 (or 1.5 if low hardness)", r=1.4}, Alpha=@n1) {constA=1, constB=-0.1}
  n6 = Constant() {desc=The radius needs to be larger if the hardness is low, r=0.5}
  n7 = LinearInterpolate(A=Constant() {r=0.99}, B=Constant() {r=-1.5}, Alpha=ScalarParameter() {parameterName=Elimination-Exit})
  n8 = FunctionInput() {bUsePreviewValueAsDefault=true, inputName=Mask-A, inputType=FunctionInput_Scalar, previewValue=<0 0 0 0>}
  n9 = SphereMask(A=@n8, B=@n7, Radius=Add(A=@n5, B=@n4), Hardness=@n6) {attenuationRadius=1, desc="! 50 perc", hardnessPercent=50}
  n10 = Multiply(A=TextureCoordinate(), B=ScalarParameter() {defaultValue=15, parameterName=NoiseScale})
  n11 = Constant() {r=1}
  n12 = ShadingPathSwitch(Default=OneMinus(None=TextureSample(UVs=@n10) {texture=/Game/Global/Textures/Default/Noise/T_FireNoise_02}.R), Forward=@n11, Mobile=@n11)
  n13 = SphereMask(A=@n8, B=@n7, Radius=@n5, Hardness=@n6) {attenuationRadius=1, desc="! 50 perc", hardnessPercent=50}
  n14 = Saturate(None=Subtract(A=Multiply(A=@n9, B=@n12), B=@n13) {constB=0.5})
  n15 = Multiply(A=@n13, B=Constant() {r=3}) {constB=5}
  n16 = Multiply(A=@n3, B=Saturate(None=Add(A=@n14, B=@n15)))
  n17 = LinearInterpolate(A=@n2.BaseColor, B=VectorParameter() {parameterName=GoldDiffuse, defaultValue=<rgba 0.5 0.2833 0.01 1>}.RGB, Alpha=@n16)
  n18 = Add(A=Multiply(A=ScalarParameter() {parameterName=ElimOutlineAmt}, B=@n13) {constB=5}, B=Multiply(A=@n14, B=Constant() {r=50}) {constB=5})
  n19 = Multiply(A=VectorParameter() {parameterName=ElimC, defaultValue=<rgba 2.5 1.1799 0.05 1>}.RGB, B=@n18)
  n20 = VectorParameter() {parameterName=Layer0-Color, defaultValue=<rgba 0 0 0 1>}
  n21 = TextureObject() {samplerType=SAMPLERTYPE_LinearColor, texture=/Game/Characters/Player/Male/Medium/Bodies/M_MED_Tailor/Textures/T_Tailor_Gradient}
  n22 = MaterialFunctionCall!(GradientPath=@n8, Noise=@n12, Gradient=@n21, LayerAlpha=ScalarParameter() {defaultValue=1.8, parameterName=Layer0-Start}) {materialFunction=&/MF_Tailor_Reactive_Layers}
  n23 = MaterialFunctionCall(GradientPath=@n8, Noise=@n12, Gradient=@n21, LayerAlpha=ScalarParameter() {defaultValue=1.8, parameterName=Layer1-Start}) {materialFunction=&MF_Tailor_Reactive_Layers}
  n24 = MaterialFunctionCall(GradientPath=@n8, Noise=@n12, Gradient=@n21, LayerAlpha=ScalarParameter() {defaultValue=1.8, parameterName=Layer2-Start}) {materialFunction=&MF_Tailor_Reactive_Layers}
  n25 = LinearInterpolate(
   A=LinearInterpolate(
    A=LinearInterpolate(A=@n19, B=@n20.RGB, Alpha=@n22.LayerColorMask)
    B=VectorParameter() {parameterName=Layer1-Color, defaultValue=<rgba 0 0 0 1>}.RGB
    Alpha=@n23.LayerColorMask
   )
   B=VectorParameter() {parameterName=Layer2-Color, defaultValue=<rgba 0 0 0 1>}.RGB
   Alpha=@n24.LayerColorMask
  )
  n26 = FunctionInput() {bUsePreviewValueAsDefault=true, inputName=Mask-G, inputType=FunctionInput_Scalar, previewValue=<0 0 0 0>}
  n27 = Multiply(A=@n26, B=@n3)
  n28 = OneMinus(None=@n16)
  n29 = Add(A=Constant() {r=0.25}, B=@n28)
  n30 = ScalarParameter() {defaultValue=0.9, parameterName=MetalAmt}
  n31 = SetMaterialAttributes!(
   MaterialAttributes=@n2.MaterialAttributes
   BaseColor=@n17
   EmissiveColor=Add(A=@n2.EmissiveColor, B=Multiply(A=@n27, B=@n25))
   Roughness=Multiply(A=@n2.Roughness, B=Saturate(None=@n29))
   Metallic=Max(A=@n2.Metallic, B=Multiply(A=@n30, B=@n16))
   SubsurfaceColor=Multiply(A=@n2.SubsurfaceColor, B=@n28)
  ) {attributeSetTypes [
    #1
    #5
    #3
    #4
    #6
   ]}
  FunctionOutput(None=@n31)
  n32 = Saturate(None=@n18)
  n33 = Max(A=Max(A=Max(A=@n32, B=@n22.Clamped), B=@n23.Clamped), B=@n24.Clamped)
  Multiply(A=@n17, B=OneMinus(None=Multiply(A=@n27, B=@n33)))
  n34 = Multiply(A=ScalarParameter() {defaultValue=-0.3, parameterName=hr564.}, B=Time())
  n35 = LinearInterpolate(A=ConstantBiasScale(None=Sine(None=@n34)), B=ScalarParameter() {parameterName=ur64443}, Alpha=ScalarParameter() {parameterName=asdg45t})
  Multiply(A=@n35, B=ScalarParameter() {defaultValue=1, parameterName=gh435ye5y34})
  Multiply(B=@n35)
  n36 = ScalarParameter() {parameterName=Layer0-InnerMod}
  n37 = ScalarParameter() {defaultValue=0.1, desc="Elimination-Rad should use the\r\nnegative of this as the min", parameterName=Layer0-RadEdge}
  n38 = LinearInterpolate(
   A=Multiply(A=Constant() {r=-1}, B=@n37)
   B=ScalarParameter() {defaultValue=1.4, desc="0 to 1 (or 1.5 if low hardness)", parameterName=Layer0-Rad}
  ) {constA=1, constB=-0.1}
  n39 = LinearInterpolate(A=Constant() {r=0.99}, B=Constant() {r=-1.5})
  n40 = SphereMask(
   A=@n8
   B=@n39
   Radius=@n38
   Hardness=ScalarParameter() {defaultValue=0.5, desc=The radius needs to be larger if the hardness is low, parameterName=Layer0-Hard}
  ) {attenuationRadius=1, desc="! 50 perc", hardnessPercent=50}
  n41 = Add(A=@n38, B=@n37)
  n42 = ScalarParameter() {defaultValue=0.5, parameterName=Layer0-EdgeHard}
  n43 = Subtract(
   A=Multiply(A=SphereMask(A=@n8, B=@n39, Radius=@n41, Hardness=@n42) {attenuationRadius=1, desc="! 50 perc", hardnessPercent=50}, B=@n12)
   B=@n40
  ) {constB=0.5}
  Saturate(None=Add(A=Multiply(A=@n36, B=@n40) {constB=5}, B=Saturate(None=@n43)))
  ScalarParameter() {parameterName=Layer0-Exit-DEBUG}
  n44 = ScalarParameter() {defaultValue=-0.3, parameterName=hr564.}
  n45 = Time()
  n46 = Frac(None=Add(A=Constant() {r=1}, B=Multiply(A=@n44, B=@n45)))
  SphereMask(A=@n8, B=Multiply(A=Subtract(A=@n46) {constB=0.2}) {constB=10}) {attenuationRadius=1, desc="! 50 perc", hardnessPercent=50}
  SphereMask(
   A=@n8
   B=LinearInterpolate(A=ScalarParameter() {defaultValue=1.5, parameterName=Layer0-A-DE}, B=ScalarParameter() {defaultValue=-0.5, parameterName=Layer0-B-DE})
   Radius=ScalarParameter() {defaultValue=0.5, parameterName=Layer0-Rad-DE}
   Hardness=ScalarParameter() {parameterName=Layer0-Hard-DE}
  ) {attenuationRadius=1, desc="! 50 perc", hardnessPercent=50}
 ]}
