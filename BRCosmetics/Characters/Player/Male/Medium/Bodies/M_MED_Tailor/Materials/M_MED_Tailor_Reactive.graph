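# Master character material for the Tailor male body cosmetic, adding a reactive effect to the skinned mesh.
# The PreFX character function feeds ClothFuzz, then MF_Tailor_Reactive, then the PostFX character function.
# An FX texture masks the layers: red drives the fuzz alpha, green drives all three connected reactive masks.
# Material attributes select between Default Lit, Subsurface Profile skin and Cloth shading models.
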Material /BRCosmetics/Characters/Player/Male/Medium/Bodies/M_MED_Tailor/Materials/M_MED_Tailor_Reactive
graph {
 properties {
  bAutomaticallySetUsageInEditor=false
  bUseMaterialAttributes=true
  bUsedWithMeshDeformer=true
  bUsedWithMorphTargets=true
  bUsedWithSkeletalMesh=true
  pixelDepthOffsetMode=PDOM_Legacy
  shadingModel=MSM_FromMaterialExpression
  subsurfaceProfile=&SS_HeroSkin_02
  usedShadingModels=Default Lit | Subsurface Profile | Cloth
 }
 nodes [
  n1 = MaterialFunctionCall() {materialFunction=&BaseFN_CharacterFunction_PreFX}
  n2 = TextureSample() {texture=/BRCosmetics/Characters/Player/Male/Medium/Bodies/M_MED_Tailor/Textures/T_M_MED_Tailor_Body_FX}
  MaterialFunctionCall(
   AO=@n1.AO
   Distance Blend=@n1.AO
   Fresnel=@n1.AO
   Material=MaterialFunctionCall!(
    Mask-G=@n2.G
    Mask-B=@n2.G
    Mask-A=@n2.G
    Material=MaterialFunctionCall!(Alpha=@n2.R, Gmap=@n1.Gmap, Material=@n1.Gmap) {materialFunction=/Game/Weapons/WeaponSkins/MaterialFunction/ClothFuzz}.Result
   ) {materialFunction=&/MF_Tailor_Reactive}.Result
  ) {materialFunction=&BaseFN_CharacterFunction_PostFX}
 ]
}
