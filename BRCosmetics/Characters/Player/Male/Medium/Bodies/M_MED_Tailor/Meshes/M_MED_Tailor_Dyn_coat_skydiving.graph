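# Physics asset for the dynamic coat on the male medium Tailor character, named as the skydiving variant.
# Six box-body chains (front, side, back per side) hang from MSTR capsules joined to spine_04, with mostly tight angular limits that loosen near the top.
# Capsules cover pelvis, spine, thighs and calves.
# Angular-free cross-constraints tie neighbouring chains mid-coat; only the right side adds lower side-to-back links and has enlarged lower side boxes.
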PhysicsAsset /BRCosmetics/Characters/Player/Male/Medium/Bodies/M_MED_Tailor/Meshes/M_MED_Tailor_Dyn_coat_skydiving
bodies {
 calf_l [{shapeType=Capsule, length=37.3125, radius=6.3438, center=<-19.8659 -1.9581 -0.0001>, rotation=<rot 90 0 0>}]
 calf_r [{shapeType=Capsule, length=37.3125, radius=6.3438, center=<19.8659 1.9581 0.0001>, rotation=<rot 90 0 0>}]
 dyn_coat_bk_l_01 [{shapeType=Box, extentX=14.5625, extentY=15.25, extentZ=1.5, center=<6 0 0>}]
 dyn_coat_bk_l_02 [{shapeType=Box, extentX=14.5625, extentY=27.6875, extentZ=1.5, center=<6 0 0>}]
 dyn_coat_bk_l_03 [{shapeType=Box, extentX=14.5625, extentY=32.4375, extentZ=2.0625, center=<6 0 0>}]
 dyn_coat_bk_l_04 [{shapeType=Box, extentX=14.5625, extentY=32.4375, extentZ=2.0625, center=<6 0 0>}]
 dyn_coat_bk_l_05 [{shapeType=Box, extentX=14.5625, extentY=32.4375, extentZ=2.0625, center=<6 0 0>}]
 dyn_coat_bk_l_06 [{shapeType=Box, extentX=14.5625, extentY=32.4375, extentZ=2.0625, center=<6 0 0>}]
 dyn_coat_bk_l_MSTR [{shapeType=Capsule, length=0.505, radius=0.505}]
 dyn_coat_bk_r_01 [{shapeType=Box, extentX=14.5625, extentY=15.25, extentZ=1.5, center=<6 0 0>}]
 dyn_coat_bk_r_02 [{shapeType=Box, extentX=14.5625, extentY=27.6875, extentZ=1.5, center=<6 0 0>}]
 dyn_coat_bk_r_03 [{shapeType=Box, extentX=14.5625, extentY=32.4375, extentZ=2.0625, center=<6 0 0>}]
 dyn_coat_bk_r_04 [{shapeType=Box, extentX=14.5625, extentY=32.4375, extentZ=2.0625, center=<6 0 0>}]
 dyn_coat_bk_r_05 [{shapeType=Box, extentX=14.5625, extentY=32.4375, extentZ=2.0625, center=<6 0 0>}]
 dyn_coat_bk_r_06 [{shapeType=Box, extentX=14.5625, extentY=32.4375, extentZ=2.0625, center=<6 0 0>}]
 dyn_coat_bk_r_MSTR [{shapeType=Capsule, length=0.505, radius=0.505}]
 dyn_coat_fr_l_01 [{shapeType=Box, extentX=14.5625, extentY=15.25, extentZ=1.5, center=<6 0 0>}]
 dyn_coat_fr_l_02 [{shapeType=Box, extentX=14.5625, extentY=28.875, extentZ=1.5, center=<6 0 0>}]
 dyn_coat_fr_l_03 [{shapeType=Box, extentX=14.5625, extentY=28.875, extentZ=2.0625, center=<6 0 0>}]
 dyn_coat_fr_l_04 [{shapeType=Box, extentX=14.5625, extentY=28.875, extentZ=2.0625, center=<6 0 0>}]
 dyn_coat_fr_l_05 [{shapeType=Box, extentX=14.5625, extentY=28.875, extentZ=2.0625, center=<6 0 0>}]
 dyn_coat_fr_l_06 [{shapeType=Box, extentX=14.5625, extentY=28.875, extentZ=4.2045, center=<6 0 0>}]
 dyn_coat_fr_l_07 [{shapeType=Box, extentX=14.5625, extentY=28.875, extentZ=4.2045, center=<6 0 0>}]
 dyn_coat_fr_l_MSTR [{shapeType=Capsule, length=0.505, radius=0.505}]
 dyn_coat_fr_r_01 [{shapeType=Box, extentX=14.5625, extentY=15.25, extentZ=1.5, center=<6 0 0>}]
 dyn_coat_fr_r_02 [{shapeType=Box, extentX=14.5625, extentY=28.875, extentZ=1.5, center=<6 0 0>}]
 dyn_coat_fr_r_03 [{shapeType=Box, extentX=14.5625, extentY=28.875, extentZ=2.0625, center=<6 0 0>}]
 dyn_coat_fr_r_04 [{shapeType=Box, extentX=14.5625, extentY=28.875, extentZ=2.0625, center=<6 0 0>}]
 dyn_coat_fr_r_05 [{shapeType=Box, extentX=14.5625, extentY=28.875, extentZ=2.0625, center=<6 0 0>}]
 dyn_coat_fr_r_06 [{shapeType=Box, extentX=14.5625, extentY=28.875, extentZ=4.2045, center=<6 0 0>}]
 dyn_coat_fr_r_07 [{shapeType=Box, extentX=14.5625, extentY=28.875, extentZ=4.2045, center=<6 0 0>}]
 dyn_coat_fr_r_MSTR [{shapeType=Capsule, length=0.505, radius=0.505}]
 dyn_coat_l_01 [{shapeType=Box, extentX=14.5625, extentY=20.125, extentZ=1.5, center=<6 0 0>}]
 dyn_coat_l_02 [{shapeType=Box, extentX=14.5625, extentY=30.9375, extentZ=1.5, center=<6 0 0>}]
 dyn_coat_l_03 [{shapeType=Box, extentX=14.5625, extentY=30.9375, extentZ=2.0625, center=<6 0 0>}]
 dyn_coat_l_04 [{shapeType=Box, extentX=14.5625, extentY=30.9375, extentZ=2.0625, center=<6 0 0>}]
 dyn_coat_l_05 [{shapeType=Box, extentX=14.5625, extentY=30.75, extentZ=2.0625, center=<6 0 0>}]
 dyn_coat_l_06 [{shapeType=Box, extentX=14.5625, extentY=30.75, extentZ=2.0625, center=<6 0 0>}]
 dyn_coat_l_MSTR [{shapeType=Capsule, length=0.505, radius=0.505}]
 dyn_coat_r_01 [{shapeType=Box, extentX=14.5625, extentY=20.125, extentZ=1.5, center=<6 0 0>}]
 dyn_coat_r_02 [{shapeType=Box, extentX=14.5625, extentY=30.9375, extentZ=1.5, center=<6 0 0>}]
 dyn_coat_r_03 [{shapeType=Box, extentX=14.5625, extentY=30.9375, extentZ=2.0625, center=<6 0 0>}]
 dyn_coat_r_04 [{shapeType=Box, extentX=14.5625, extentY=30.9375, extentZ=2.0625, center=<6 0 0>}]
 dyn_coat_r_05 [{shapeType=Box, extentX=14.5625, extentY=40.0625, extentZ=3.625, center=<6 0 0>}]
 dyn_coat_r_06 [{shapeType=Box, extentX=14.5625, extentY=40.0625, extentZ=3.625, center=<6 0 0>}]
 dyn_coat_r_MSTR [{shapeType=Capsule, length=0.505, radius=0.505}]
 pelvis [{shapeType=Capsule, length=10.25, radius=10.3125, center=<-2.6114 -2.8515 0>}]
 spine_01 [{shapeType=Capsule, length=11.0625, radius=9.375, center=<-0.743 -1.7164 0>}]
 spine_02 [{shapeType=Capsule, length=8.3438, radius=9.375, center=<-0.7431 -0.9379 0>}]
 spine_03 [{shapeType=Capsule, length=9.2812, radius=9.375, center=<-0.743 -0.6637 0>}]
 spine_04 [{shapeType=Capsule, length=11.0625, radius=9.375, center=<-0.7431 -0.6734 0.0001>}]
 thigh_l [{shapeType=Capsule, length=25.7188, radius=8.6562, center=<-18.1396 -0.8169 1.4515>, rotation=<rot 90 1.7899 -0.8059>}]
 thigh_r [{shapeType=Capsule, length=25.7188, radius=8.6562, center=<18.1396 0.8169 -1.4515>, rotation=<rot 90 1.7899 -0.8059>}]
}
constraints [
 {
  bone1Name=spine_01
  bone2Name=pelvis
  swing1LimitDegrees=45
  swing1Motion=Free
  swing2LimitDegrees=45
  swing2Motion=Free
  twistLimitDegrees=45
  twistMotion=Free
 }
 {
  bone1Name=spine_02
  bone2Name=spine_01
  swing1LimitDegrees=45
  swing1Motion=Free
  swing2LimitDegrees=45
  swing2Motion=Free
  twistLimitDegrees=45
  twistMotion=Free
 }
 {
  bone1Name=spine_03
  bone2Name=spine_02
  swing1LimitDegrees=45
  swing1Motion=Free
  swing2LimitDegrees=45
  swing2Motion=Free
  twistLimitDegrees=45
  twistMotion=Free
 }
 {
  bone1Name=spine_04
  bone2Name=spine_03
  swing1LimitDegrees=45
  swing1Motion=Free
  swing2LimitDegrees=45
  swing2Motion=Free
  twistLimitDegrees=45
  twistMotion=Free
 }
 {
  bone1Name=dyn_coat_l_MSTR
  bone2Name=spine_04
  swing1LimitDegrees=45
  swing1Motion=Free
  swing2LimitDegrees=45
  swing2Motion=Free
  twistLimitDegrees=45
  twistMotion=Free
 }
 {bone1Name=dyn_coat_l_01, bone2Name=dyn_coat_l_MSTR, swing1LimitDegrees=5, swing2LimitDegrees=40, twistLimitDegrees=5}
 {bone1Name=dyn_coat_l_02, bone2Name=dyn_coat_l_01, swing1LimitDegrees=5, swing2LimitDegrees=5, twistLimitDegrees=5}
 {bone1Name=dyn_coat_l_03, bone2Name=dyn_coat_l_02, swing1LimitDegrees=5, swing2LimitDegrees=5, twistLimitDegrees=5}
 {bone1Name=dyn_coat_l_04, bone2Name=dyn_coat_l_03, swing1LimitDegrees=5, swing2LimitDegrees=5, twistLimitDegrees=5}
 {bone1Name=dyn_coat_l_05, bone2Name=dyn_coat_l_04, swing1LimitDegrees=5, swing2LimitDegrees=5, twistLimitDegrees=5}
 {bone1Name=dyn_coat_l_06, bone2Name=dyn_coat_l_05, swing1LimitDegrees=5, swing2LimitDegrees=5, twistLimitDegrees=5}
 {
  bone1Name=dyn_coat_fr_l_MSTR
  bone2Name=spine_04
  swing1LimitDegrees=45
  swing1Motion=Free
  swing2LimitDegrees=45
  swing2Motion=Free
  twistLimitDegrees=45
  twistMotion=Free
 }
 {bone1Name=dyn_coat_fr_l_01, bone2Name=dyn_coat_fr_l_MSTR, swing1LimitDegrees=5, swing2LimitDegrees=40, twistLimitDegrees=5}
 {bone1Name=dyn_coat_fr_l_02, bone2Name=dyn_coat_fr_l_01, swing1LimitDegrees=5, swing2LimitDegrees=40, twistLimitDegrees=5}
 {bone1Name=dyn_coat_fr_l_03, bone2Name=dyn_coat_fr_l_02, swing1LimitDegrees=10, swing2LimitDegrees=15, twistLimitDegrees=5}
 {bone1Name=dyn_coat_fr_l_04, bone2Name=dyn_coat_fr_l_03, swing1LimitDegrees=5, swing2LimitDegrees=7, twistLimitDegrees=5}
 {bone1Name=dyn_coat_fr_l_05, bone2Name=dyn_coat_fr_l_04, swing1LimitDegrees=5, swing2LimitDegrees=7, twistLimitDegrees=5}
 {bone1Name=dyn_coat_fr_l_06, bone2Name=dyn_coat_fr_l_05, swing1LimitDegrees=5, swing2LimitDegrees=7, twistLimitDegrees=5}
 {bone1Name=dyn_coat_fr_l_07, bone2Name=dyn_coat_fr_l_06, swing1LimitDegrees=5, swing2LimitDegrees=7, twistLimitDegrees=5}
 {
  bone1Name=dyn_coat_bk_l_MSTR
  bone2Name=spine_04
  swing1LimitDegrees=45
  swing1Motion=Free
  swing2LimitDegrees=45
  swing2Motion=Free
  twistLimitDegrees=45
  twistMotion=Free
 }
 {bone1Name=dyn_coat_bk_l_01, bone2Name=dyn_coat_bk_l_MSTR, swing1LimitDegrees=5, swing2LimitDegrees=40, twistLimitDegrees=5}
 {bone1Name=dyn_coat_bk_l_02, bone2Name=dyn_coat_bk_l_01, swing1LimitDegrees=5, swing2LimitDegrees=5, twistLimitDegrees=5}
 {bone1Name=dyn_coat_bk_l_03, bone2Name=dyn_coat_bk_l_02, swing1LimitDegrees=5, swing2LimitDegrees=5, twistLimitDegrees=5}
 {bone1Name=dyn_coat_bk_l_04, bone2Name=dyn_coat_bk_l_03, swing1LimitDegrees=5, swing2LimitDegrees=5, twistLimitDegrees=5}
 {bone1Name=dyn_coat_bk_l_05, bone2Name=dyn_coat_bk_l_04, swing1LimitDegrees=5, swing2LimitDegrees=5, twistLimitDegrees=5}
 {bone1Name=dyn_coat_bk_l_06, bone2Name=dyn_coat_bk_l_05, swing1LimitDegrees=5, swing2LimitDegrees=5, twistLimitDegrees=5}
 {
  bone1Name=dyn_coat_bk_r_MSTR
  bone2Name=spine_04
  swing1LimitDegrees=45
  swing1Motion=Free
  swing2LimitDegrees=45
  swing2Motion=Free
  twistLimitDegrees=45
  twistMotion=Free
 }
 {bone1Name=dyn_coat_bk_r_01, bone2Name=dyn_coat_bk_r_MSTR, swing1LimitDegrees=5, swing2LimitDegrees=40, twistLimitDegrees=5}
 {bone1Name=dyn_coat_bk_r_02, bone2Name=dyn_coat_bk_r_01, swing1LimitDegrees=5, swing2LimitDegrees=5, twistLimitDegrees=5}
 {bone1Name=dyn_coat_bk_r_03, bone2Name=dyn_coat_bk_r_02, swing1LimitDegrees=5, swing2LimitDegrees=5, twistLimitDegrees=5}
 {bone1Name=dyn_coat_bk_r_04, bone2Name=dyn_coat_bk_r_03, swing1LimitDegrees=5, swing2LimitDegrees=5, twistLimitDegrees=5}
 {bone1Name=dyn_coat_bk_r_05, bone2Name=dyn_coat_bk_r_04, swing1LimitDegrees=5, swing2LimitDegrees=5, twistLimitDegrees=5}
 {bone1Name=dyn_coat_bk_r_06, bone2Name=dyn_coat_bk_r_05, swing1LimitDegrees=5, swing2LimitDegrees=5, twistLimitDegrees=5}
 {
  bone1Name=dyn_coat_r_MSTR
  bone2Name=spine_04
  swing1LimitDegrees=45
  swing1Motion=Free
  swing2LimitDegrees=45
  swing2Motion=Free
  twistLimitDegrees=45
  twistMotion=Free
 }
 {bone1Name=dyn_coat_r_01, bone2Name=dyn_coat_r_MSTR, swing1LimitDegrees=5, swing2LimitDegrees=40, twistLimitDegrees=5}
 {bone1Name=dyn_coat_r_02, bone2Name=dyn_coat_r_01, swing1LimitDegrees=5, swing2LimitDegrees=5, twistLimitDegrees=5}
 {bone1Name=dyn_coat_r_03, bone2Name=dyn_coat_r_02, swing1LimitDegrees=5, swing2LimitDegrees=5, twistLimitDegrees=5}
 {bone1Name=dyn_coat_r_04, bone2Name=dyn_coat_r_03, swing1LimitDegrees=5, swing2LimitDegrees=5, twistLimitDegrees=5}
 {bone1Name=dyn_coat_r_05, bone2Name=dyn_coat_r_04, swing1LimitDegrees=5, swing2LimitDegrees=5, twistLimitDegrees=5}
 {bone1Name=dyn_coat_r_06, bone2Name=dyn_coat_r_05, swing1LimitDegrees=5, swing2LimitDegrees=5, twistLimitDegrees=5}
 {
  bone1Name=dyn_coat_fr_r_MSTR
  bone2Name=spine_04
  swing1LimitDegrees=45
  swing1Motion=Free
  swing2LimitDegrees=45
  swing2Motion=Free
  twistLimitDegrees=45
  twistMotion=Free
 }
 {bone1Name=dyn_coat_fr_r_01, bone2Name=dyn_coat_fr_r_MSTR, swing1LimitDegrees=5, swing2LimitDegrees=40, twistLimitDegrees=5}
 {bone1Name=dyn_coat_fr_r_02, bone2Name=dyn_coat_fr_r_01, swing1LimitDegrees=5, swing2LimitDegrees=40, twistLimitDegrees=5}
 {bone1Name=dyn_coat_fr_r_03, bone2Name=dyn_coat_fr_r_02, swing1LimitDegrees=10, swing2LimitDegrees=15, twistLimitDegrees=5}
 {bone1Name=dyn_coat_fr_r_04, bone2Name=dyn_coat_fr_r_03, swing1LimitDegrees=5, swing2LimitDegrees=7, twistLimitDegrees=5}
 {bone1Name=dyn_coat_fr_r_05, bone2Name=dyn_coat_fr_r_04, swing1LimitDegrees=5, swing2LimitDegrees=7, twistLimitDegrees=5}
 {bone1Name=dyn_coat_fr_r_06, bone2Name=dyn_coat_fr_r_05, swing1LimitDegrees=5, swing2LimitDegrees=7, twistLimitDegrees=5}
 {bone1Name=dyn_coat_fr_r_07, bone2Name=dyn_coat_fr_r_06, swing1LimitDegrees=5, swing2LimitDegrees=7, twistLimitDegrees=5}
 {
  bone1Name=thigh_l
  bone2Name=pelvis
  swing1LimitDegrees=45
  swing1Motion=Free
  swing2LimitDegrees=45
  swing2Motion=Free
  twistLimitDegrees=45
  twistMotion=Free
 }
 {
  bone1Name=thigh_r
  bone2Name=pelvis
  swing1LimitDegrees=45
  swing1Motion=Free
  swing2LimitDegrees=45
  swing2Motion=Free
  twistLimitDegrees=45
  twistMotion=Free
 }
 {
  bone1Name=calf_l
  bone2Name=thigh_l
  swing1LimitDegrees=45
  swing1Motion=Free
  swing2LimitDegrees=45
  swing2Motion=Free
  twistLimitDegrees=45
  twistMotion=Free
 }
 {
  bone1Name=calf_r
  bone2Name=thigh_r
  swing1LimitDegrees=45
  swing1Motion=Free
  swing2LimitDegrees=45
  swing2Motion=Free
  twistLimitDegrees=45
  twistMotion=Free
 }
 {
  bone1Name=dyn_coat_bk_r_03
  bone2Name=dyn_coat_r_03
  swing1LimitDegrees=45
  swing1Motion=Free
  swing2LimitDegrees=45
  swing2Motion=Free
  twistLimitDegrees=45
  twistMotion=Free
 }
 {
  bone1Name=dyn_coat_bk_l_03
  bone2Name=dyn_coat_l_03
  swing1LimitDegrees=45
  swing1Motion=Free
  swing2LimitDegrees=45
  swing2Motion=Free
  twistLimitDegrees=45
  twistMotion=Free
 }
 {
  bone1Name=dyn_coat_r_03
  bone2Name=dyn_coat_fr_r_04
  swing1LimitDegrees=45
  swing1Motion=Free
  swing2LimitDegrees=45
  swing2Motion=Free
  twistLimitDegrees=45
  twistMotion=Free
 }
 {
  bone1Name=dyn_coat_l_03
  bone2Name=dyn_coat_fr_l_04
  swing1LimitDegrees=45
  swing1Motion=Free
  swing2LimitDegrees=45
  swing2Motion=Free
  twistLimitDegrees=45
  twistMotion=Free
 }
 {
  bone1Name=dyn_coat_r_05
  bone2Name=dyn_coat_bk_r_05
  swing1LimitDegrees=45
  swing1Motion=Free
  swing2LimitDegrees=45
  swing2Motion=Free
  twistLimitDegrees=45
  twistMotion=Free
 }
 {
  bone1Name=dyn_coat_r_06
  bone2Name=dyn_coat_bk_r_06
  swing1LimitDegrees=45
  swing1Motion=Free
  swing2LimitDegrees=45
  swing2Motion=Free
  twistLimitDegrees=45
  twistMotion=Free
 }
]
