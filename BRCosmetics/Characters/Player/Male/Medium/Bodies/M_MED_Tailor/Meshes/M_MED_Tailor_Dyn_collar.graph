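# Physics asset for the collar and lapels of the M_MED_Tailor male player body, defining the bodies and limits for their secondary motion.
# Capsules cover the back, rear-side and side collar bones, boxes cover both lapels, and a head capsule has no constraint.
# Each collar or lapel bone is constrained to its own tiny _MSTR capsule.
# Swing and twist limits are mirrored left to right, with the lapels tightest in first swing and twist.
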PhysicsAsset /BRCosmetics/Characters/Player/Male/Medium/Bodies/M_MED_Tailor/Meshes/M_MED_Tailor_Dyn_collar
bodies {
 dyn_collar_back [{shapeType=Capsule, length=12, radius=5, shapeName=dyn_collar_back_capsule, center=<5 0 0>, rotation=<rot 0 0 90>}]
 dyn_collar_back_MSTR [{shapeType=Capsule, length=0.505, radius=0.505}]
 dyn_collar_bk_l [{shapeType=Capsule, length=6.3114, radius=3, shapeName=dyn_collar_bk_l_capsule, center=<3 0 0>, rotation=<rot 90 0 0>}]
 dyn_collar_bk_l_MSTR [{shapeType=Capsule, length=0.505, radius=0.505}]
 dyn_collar_bk_r [{shapeType=Capsule, length=5.1602, radius=3, shapeName=dyn_collar_bk_r_capsule, center=<3 0 0>, rotation=<rot 90 0 0>}]
 dyn_collar_bk_r_MSTR [{shapeType=Capsule, length=0.505, radius=0.505}]
 dyn_collar_l [{shapeType=Capsule, length=10, radius=2.5, shapeName=dyn_collar_l_capsule, center=<5 0 0>, rotation=<rot 90 0 0>}]
 dyn_collar_l_MSTR [{shapeType=Capsule, length=0.505, radius=0.505}]
 dyn_collar_r [{shapeType=Capsule, length=10, radius=2.5, shapeName=dyn_collar_r_capsule, center=<5 0 0>, rotation=<rot 90 0 0>}]
 dyn_collar_r_MSTR [{shapeType=Capsule, length=0.505, radius=0.505}]
 dyn_lapel_l [{shapeType=Box, extentX=6.6875, extentY=8.4375, extentZ=1.3125, center=<2.2 0 0>}]
 dyn_lapel_l_MSTR [{shapeType=Capsule, length=0.505, radius=0.505}]
 dyn_rapel_r [{shapeType=Box, extentX=6.6875, extentY=8.4375, extentZ=1.3125, center=<2.2 0 0>}]
 dyn_rapel_r_MSTR [{shapeType=Capsule, length=0.505, radius=0.505}]
 head [{shapeType=Capsule, length=9.7926, radius=6.6838, center=<3.4054 4.6812 -0.0001>, rotation=<rot 90 90 96.5633>}]
}
constraints [
 {bone1Name=dyn_collar_l, bone2Name=dyn_collar_l_MSTR, swing1LimitDegrees=45, swing2LimitDegrees=45, twistLimitDegrees=15}
 {bone1Name=dyn_collar_bk_l, bone2Name=dyn_collar_bk_l_MSTR, swing1LimitDegrees=80, swing2LimitDegrees=45, twistLimitDegrees=15}
 {bone1Name=dyn_collar_back, bone2Name=dyn_collar_back_MSTR, swing1LimitDegrees=80, swing2LimitDegrees=45, twistLimitDegrees=15}
 {bone1Name=dyn_lapel_l, bone2Name=dyn_lapel_l_MSTR, swing1LimitDegrees=5, swing2LimitDegrees=45, twistLimitDegrees=5}
 {bone1Name=dyn_collar_bk_r, bone2Name=dyn_collar_bk_r_MSTR, swing1LimitDegrees=80, swing2LimitDegrees=45, twistLimitDegrees=15}
 {bone1Name=dyn_collar_r, bone2Name=dyn_collar_r_MSTR, swing1LimitDegrees=45, swing2LimitDegrees=45, twistLimitDegrees=15}
 {bone1Name=dyn_rapel_r, bone2Name=dyn_rapel_r_MSTR, swing1LimitDegrees=5, swing2LimitDegrees=45, twistLimitDegrees=5}
]
