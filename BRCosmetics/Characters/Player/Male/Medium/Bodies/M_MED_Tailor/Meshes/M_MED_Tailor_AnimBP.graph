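# Animation Blueprint for the Tailor body cosmetic that copies its attached parent's pose and adds coat and collar secondary motion.
# Thigh-driven corrections, vehicle-driver coat shrinking and collar physics feed a cached pose.
# That pose branches into standard or skydiving coat RigidBody simulations with bone constraints and offsets, blended by skydiving state.
# The event graph drives gravity and alphas from swim, skydive, vehicle, crouch-slide, thigh and emote/front-end states.
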AnimBlueprint /BRCosmetics/Characters/Player/Male/Medium/Bodies/M_MED_Tailor/Meshes/M_MED_Tailor_AnimBP
target_skeleton=&/M_MED_Tailor_Skeleton
graphs [
 AnimGraph {nodes [
   n1 = RigidBody!(
    ComponentPose=Local To Component(LocalPose=Use cached pose 'main_SavedPose'())
    Alpha=Get in vehicle rigid alpha() -> (in vehicle rigid alpha:double)
    OverrideWorldGravity:Vector=Get gravity_override() -> (gravity_override:Vector)
    ExternalForce
   ) -> (Pose) {
    bClampLinearTranslationLimitToRefPose=true
    bOverrideWorldGravity=true
    overridePhysicsAsset=&/M_MED_Tailor_Dyn_coat
    simulationSpace=BaseBoneSpace
    alphaBoolBlend {blendOption=Linear}
    baseBoneRef {boneName=spine_02}
    componentAppliedLinearAccClamp=<3000 3000 3000>
    componentLinearAccScale=<0.4 0.3 0.3>
    componentLinearVelScale=<0.4 0.3 0.3>
    overrideWorldGravity {z=-980}
   }
   n2 = "Constraint - dyn_collar_back_MSTR (P)"(ComponentPose=@n1, Alpha=0.35, ConstraintWeights_0=1) {
    alphaBoolBlend {blendOption=Linear}
    boneToModify {boneName=dyn_collar_back_MSTR}
    constraintSetup [{targetBone {boneName=head}}]
    constraintWeights [1]
   }
   n3 = "Constraint - dyn_collar_bk_l_MSTR (P)"(ComponentPose=@n2, Alpha=0.35, ConstraintWeights_0=1) {
    alphaBoolBlend {blendOption=Linear}
    boneToModify {boneName=dyn_collar_bk_l_MSTR}
    constraintSetup [{targetBone {boneName=head}}]
    constraintWeights [1]
   }
   n4 = "Constraint - dyn_collar_bk_r_MSTR (P)"(ComponentPose=@n3, Alpha=0.35, ConstraintWeights_0=1) {
    alphaBoolBlend {blendOption=Linear}
    boneToModify {boneName=dyn_collar_bk_r_MSTR}
    constraintSetup [{targetBone {boneName=head}}]
    constraintWeights [1]
   }
   n5 = "Constraint - dyn_coat_bk_04 (PP)"!(ComponentPose=@n4, ConstraintWeights_0=0.5, ConstraintWeights_1:float=0.5) -> (Pose) {
    alphaBoolBlend {blendOption=Linear}
    boneToModify {boneName=dyn_coat_bk_04}
    constraintSetup [{targetBone {boneName=dyn_coat_bk_l_04}}, {targetBone {boneName=dyn_coat_bk_r_04}}]
    constraintWeights [1,1]
   }
   n6 = "Constraint - dyn_coat_bk_05 (PP)"!(ComponentPose=@n5, ConstraintWeights_0=0.5, ConstraintWeights_1:float=0.5) -> (Pose) {
    alphaBoolBlend {blendOption=Linear}
    boneToModify {boneName=dyn_coat_bk_05}
    constraintSetup [{targetBone {boneName=dyn_coat_bk_l_05}}, {targetBone {boneName=dyn_coat_bk_r_05}}]
    constraintWeights [1,1]
   }
   n7 = "Constraint - dyn_coat_bk_06 (PP)"!(ComponentPose=@n6, ConstraintWeights_0=0.5, ConstraintWeights_1:float=0.5) -> (Pose) {
    alphaBoolBlend {blendOption=Linear}
    boneToModify {boneName=dyn_coat_bk_06}
    constraintSetup [{targetBone {boneName=dyn_coat_bk_l_06}}, {targetBone {boneName=dyn_coat_bk_r_06}}]
    constraintWeights [1,1]
   }
   n8 = Get isCrouchSliding() -> (isCrouchSliding:double)
   n9 = "Transform (Modify) Bone - Bone: dyn_coat_bk_r_03"(ComponentPose=@n7, Alpha=@n8, Translation="0,0,0", Rotation="0,0,45", Scale="1,1,1") {rotationMode=BMM_Additive, alphaBoolBlend {blendOption=Linear}, boneToModify {boneName=dyn_coat_bk_r_03}, rotation {roll=45}}
   n10 = "Transform (Modify) Bone - Bone: dyn_coat_bk_l_03"(ComponentPose=@n9, Alpha=@n8, Translation="0,0,0", Rotation="0,0,45", Scale="1,1,1") {rotationMode=BMM_Additive, alphaBoolBlend {blendOption=Linear}, boneToModify {boneName=dyn_coat_bk_l_03}, rotation {roll=45}}
   n11 = "Transform (Modify) Bone - Bone: dyn_coat_bk_03"(ComponentPose=@n10, Alpha=@n8, Translation="0,0,0", Rotation="0,0,30", Scale="1,1,1") {rotationMode=BMM_Additive, alphaBoolBlend {blendOption=Linear}, boneToModify {boneName=dyn_coat_bk_03}, rotation {roll=30}}
   n12 = "Transform (Modify) Bone - Bone: dyn_coat_r_03"(ComponentPose=@n11, Alpha=@n8, Translation="0,0,0", Rotation="-60,0,20", Scale="1,1,1") {rotationMode=BMM_Additive, alphaBoolBlend {blendOption=Linear}, boneToModify {boneName=dyn_coat_r_03}, rotation {pitch=-60, roll=20}}
   n13 = "Transform (Modify) Bone - Bone: dyn_coat_l_03"(ComponentPose=@n12, Alpha=@n8, Translation="0,0,0", Rotation="60,0,20", Scale="1,1,1") {rotationMode=BMM_Additive, alphaBoolBlend {blendOption=Linear}, boneToModify {boneName=dyn_coat_l_03}, rotation {pitch=60, roll=20}}
   n14 = "Transform (Modify) Bone - Bone: dyn_coat_fr_l_01"(ComponentPose=@n13, Alpha=Get LegRot_l() -> (LegRot_l:double), Translation="-5,0,0", Rotation="10,0,-10", Scale="1,1,1") {
    rotationMode=BMM_Additive
    translationMode=BMM_Additive
    translationSpace=BCS_BoneSpace
    alphaBoolBlend {blendOption=Linear}
    alphaScaleBiasClamp {bInterpResult=true, interpSpeedDecreasing=5, interpSpeedIncreasing=5}
    boneToModify {boneName=dyn_coat_fr_l_01}
    rotation {pitch=10, roll=-10}
    translation {x=-5}
   }
   n15 = "Transform (Modify) Bone - Bone: dyn_coat_fr_r_02"(ComponentPose=@n14, Alpha=Get LegRot_r() -> (LegRot_r:double), Translation="-5,0,0", Rotation="-10,0,10", Scale="1,1,1") {
    rotationMode=BMM_Additive
    translationMode=BMM_Additive
    translationSpace=BCS_BoneSpace
    alphaBoolBlend {blendOption=Linear}
    alphaScaleBiasClamp {bInterpResult=true, interpSpeedDecreasing=8, interpSpeedIncreasing=8}
    boneToModify {boneName=dyn_coat_fr_r_02}
    rotation {pitch=-10, roll=10}
    translation {x=-5}
   }
   n16 = "Transform (Modify) Bone - Bone: dyn_coat_fr_r_02"(ComponentPose=@n15, Alpha=Get LegBackRot_r() -> (LegBackRot_r:double), Translation="0,10,5", Rotation="-10,0,10", Scale="1,1,1") {
    rotationMode=BMM_Additive
    translationMode=BMM_Additive
    translationSpace=BCS_BoneSpace
    alphaBoolBlend {blendOption=Linear}
    alphaScaleBiasClamp {bInterpResult=true, interpSpeedDecreasing=8, interpSpeedIncreasing=8}
    boneToModify {boneName=dyn_coat_fr_r_02}
    rotation {pitch=-10, roll=10}
    translation {y=10, z=5}
   }
   n17 = RigidBody(ComponentPose=Local To Component(LocalPose=Use cached pose 'main_SavedPose'()), OverrideWorldGravity=Get gravity_override()) {
    bClampLinearTranslationLimitToRefPose=true
    bOverrideWorldGravity=true
    overridePhysicsAsset=&/M_MED_Tailor_Dyn_coat_skydiving
    simulationSpace=BaseBoneSpace
    alphaBoolBlend {blendOption=Linear}
    baseBoneRef {boneName=spine_02}
    componentAppliedLinearAccClamp=<3000 3000 3000>
    componentLinearAccScale=<0.4 0.3 0.3>
    componentLinearVelScale=<0.4 0.3 0.3>
    overrideWorldGravity {z=-980}
   }
   n18 = "Constraint - dyn_collar_back_MSTR (P)"(ComponentPose=@n17, Alpha=0.3, ConstraintWeights_0=1) {
    alphaBoolBlend {blendOption=Linear}
    boneToModify {boneName=dyn_collar_back_MSTR}
    constraintSetup [{targetBone {boneName=head}}]
    constraintWeights [1]
   }
   n19 = "Constraint - dyn_coat_bk_02 (PP)"!(ComponentPose=@n18, Alpha=1, ConstraintWeights_0=0.5, ConstraintWeights_1:float=0.5) -> (Pose) {
    alphaBoolBlend {blendOption=Linear}
    boneToModify {boneName=dyn_coat_bk_02}
    constraintSetup [{targetBone {boneName=dyn_coat_bk_l_02}}, {targetBone {boneName=dyn_coat_bk_r_02}}]
    constraintWeights [1,1]
   }
   n20 = "Constraint - dyn_coat_bk_03 (PP)"!(ComponentPose=@n19, Alpha=1, ConstraintWeights_0=0.5, ConstraintWeights_1:float=0.5) -> (Pose) {
    alphaBoolBlend {blendOption=Linear}
    boneToModify {boneName=dyn_coat_bk_03}
    constraintSetup [{targetBone {boneName=dyn_coat_bk_l_03}}, {targetBone {boneName=dyn_coat_bk_r_03}}]
    constraintWeights [1,1]
   }
   n21 = "Constraint - dyn_coat_bk_04 (PP)"!(ComponentPose=@n20, Alpha:float=1, ConstraintWeights_0=0.5, ConstraintWeights_1=0.5) -> (Pose) {
    alphaBoolBlend {blendOption=Linear}
    boneToModify {boneName=dyn_coat_bk_04}
    constraintSetup [{targetBone {boneName=dyn_coat_bk_l_04}}, {targetBone {boneName=dyn_coat_bk_r_04}}]
    constraintWeights [1,1]
   }
   n22 = "Constraint - dyn_coat_bk_05 (PP)"!(ComponentPose=@n21, Alpha:float=1, ConstraintWeights_0=0.5, ConstraintWeights_1=0.5) -> (Pose) {
    alphaBoolBlend {blendOption=Linear}
    boneToModify {boneName=dyn_coat_bk_05}
    constraintSetup [{targetBone {boneName=dyn_coat_bk_l_05}}, {targetBone {boneName=dyn_coat_bk_r_05}}]
    constraintWeights [1,1]
   }
   n23 = "Constraint - dyn_coat_bk_06 (PP)"!(ComponentPose=@n22, Alpha:float=1, ConstraintWeights_0=0.5, ConstraintWeights_1=0.5) -> (Pose) {
    alphaBoolBlend {blendOption=Linear}
    boneToModify {boneName=dyn_coat_bk_06}
    constraintSetup [{targetBone {boneName=dyn_coat_bk_l_06}}, {targetBone {boneName=dyn_coat_bk_r_06}}]
    constraintWeights [1,1]
   }
   n24 = "Transform (Modify) Bone - Bone: dyn_coat_bk_l_01"(ComponentPose=@n23, Alpha=1, Translation="0,0,0", Rotation="15,0,15", Scale="1,1,1") {rotationMode=BMM_Additive, alphaBoolBlend {blendOption=Linear}, boneToModify {boneName=dyn_coat_bk_l_01}, rotation {pitch=15, roll=15}}
   n25 = "Transform (Modify) Bone - Bone: dyn_coat_bk_r_01"(ComponentPose=@n24, Alpha=1, Translation="0,0,0", Rotation="-15,0,15", Scale="1,1,1") {rotationMode=BMM_Additive, alphaBoolBlend {blendOption=Linear}, boneToModify {boneName=dyn_coat_bk_r_01}, rotation {pitch=-15, roll=15}}
   Output Pose(Result=Blend Poses by bool(BlendPose_0=Component To Local(ComponentPose=@n25), BlendPose_1=Component To Local(ComponentPose=@n16), bActiveValue=Get bIsSkyDiving()) {blendType=Linear}) {layerGroup=None}
   n26 = "Transform (Modify) Bone - Bone: dyn_coat_fr_l_MSTR"(
    ComponentPose=Local To Component(LocalPose=Copy Pose From Mesh() {bUseAttachedParent=true})
    Alpha=Get thigh_rot_l()
    Translation="0,0,0"
    Rotation="3.870,-45.640,-3.944"
    Scale="1,1,1"
   ) {
    rotationMode=BMM_Additive
    rotationSpace=BCS_BoneSpace
    translationMode=BMM_Additive
    translationSpace=BCS_BoneSpace
    alphaBoolBlend {blendOption=Linear}
    alphaScaleBiasClamp {bInterpResult=true, interpSpeedDecreasing=15, interpSpeedIncreasing=15}
    boneToModify {boneName=dyn_coat_fr_l_MSTR}
    rotation=<rot 3.87 -45.6401 -3.944>
   }
   n27 = "Transform (Modify) Bone - Bone: dyn_coat_fr_r_MSTR"(ComponentPose=@n26, Alpha=Get thigh_rot_r(), Translation="0,0,0", Rotation="0,37.71,0", Scale="1,1,1") {
    rotationMode=BMM_Additive
    rotationSpace=BCS_BoneSpace
    translationMode=BMM_Additive
    translationSpace=BCS_BoneSpace
    alphaBoolBlend {blendOption=Linear}
    alphaScaleBiasClamp {bInterpResult=true, interpSpeedDecreasing=15, interpSpeedIncreasing=15}
    boneToModify {boneName=dyn_coat_fr_r_MSTR}
    rotation {yaw=37.71}
   }
   n28 = Get in vehicle scale alpha() -> (in vehicle scale alpha:double)
   n29 = "Transform (Modify) Bone - Bone: dyn_coat_l_02"(ComponentPose=@n27, Alpha=@n28, Translation="0,0,0", Rotation="0,0,0", Scale="0.1,0.1,0.1") {scaleMode=BMM_Replace, alphaBoolBlend {blendOption=Linear}, boneToModify {boneName=dyn_coat_l_02}, scale=<0.1 0.1 0.1>}
   n30 = "Transform (Modify) Bone - Bone: dyn_coat_fr_l_02"(ComponentPose=@n29, Alpha=@n28, Translation="0,0,0", Rotation="0,0,0", Scale="0.1,0.1,0.1") {scaleMode=BMM_Replace, alphaBoolBlend {blendOption=Linear}, boneToModify {boneName=dyn_coat_fr_l_02}, scale=<0.1 0.1 0.1>}
   n31 = "Transform (Modify) Bone - Bone: dyn_coat_bk_02"(ComponentPose=@n30, Alpha=@n28, Translation="0,0,0", Rotation="0,0,0", Scale="0.1,0.1,0.1") {scaleMode=BMM_Replace, alphaBoolBlend {blendOption=Linear}, boneToModify {boneName=dyn_coat_bk_02}, scale=<0.1 0.1 0.1>}
   n32 = "Transform (Modify) Bone - Bone: dyn_coat_bk_l_02"(ComponentPose=@n31, Alpha=@n28, Translation="0,0,0", Rotation="0,0,0", Scale="0.1,0.1,0.1") {scaleMode=BMM_Replace, alphaBoolBlend {blendOption=Linear}, boneToModify {boneName=dyn_coat_bk_l_02}, scale=<0.1 0.1 0.1>}
   n33 = "Transform (Modify) Bone - Bone: dyn_coat_bk_r_02"(ComponentPose=@n32, Alpha=@n28, Translation="0,0,0", Rotation="0,0,0", Scale="0.1,0.1,0.1") {scaleMode=BMM_Replace, alphaBoolBlend {blendOption=Linear}, boneToModify {boneName=dyn_coat_bk_r_02}, scale=<0.1 0.1 0.1>}
   n34 = "Transform (Modify) Bone - Bone: dyn_coat_fr_r_02"(ComponentPose=@n33, Alpha=@n28, Translation="0,0,0", Rotation="0,0,0", Scale="0.1,0.1,0.1") {scaleMode=BMM_Replace, alphaBoolBlend {blendOption=Linear}, boneToModify {boneName=dyn_coat_fr_r_02}, scale=<0.1 0.1 0.1>}
   n35 = "Transform (Modify) Bone - Bone: dyn_coat_r_02"(ComponentPose=@n34, Alpha=@n28, Translation="0,0,0", Rotation="0,0,0", Scale="0.1,0.1,0.1") {scaleMode=BMM_Replace, alphaBoolBlend {blendOption=Linear}, boneToModify {boneName=dyn_coat_r_02}, scale=<0.1 0.1 0.1>}
   n36 = RigidBody!(ComponentPose=@n35, Alpha, ExternalForce) -> (Pose) {
    bClampLinearTranslationLimitToRefPose=true
    bOverrideWorldGravity=true
    overridePhysicsAsset=&/M_MED_Tailor_Dyn_collar
    simulationSpace=BaseBoneSpace
    alphaBoolBlend {blendOption=Linear}
    baseBoneRef {boneName=spine_05}
    componentAppliedLinearAccClamp=<3000 3000 3000>
    componentLinearAccScale=<0.5 0.4 0.4>
    componentLinearVelScale=<0.5 0.4 0.4>
    overrideWorldGravity {z=-980}
   }
   Save cached pose 'main_SavedPose'(Pose=Component To Local(ComponentPose=@n36)) {cacheName=main_SavedPose}
  ]}
 EventGraph {nodes [
   Event BlueprintUpdateAnimation()
   > n1 = Sequence!(execute) -> (then_0, then_1, then_2:Exec, then_3:Exec, then_4:Exec)
   > n2 = Branch(Condition=Get bIsSwimSprinting() -> (bIsSwimSprinting:bool))
   > Set gravity_override(gravity_override="0, 0, 0")
   n3 = Branch(execute=@n2.else, Condition=Get bIsSkyDiving())
   > n4 = Branch(Condition=Get bIsParachuteOpened())
   Set gravity_override(execute=[@n4.then, @n3.else], gravity_override="0, 0,-980")
   Set gravity_override(execute=@n4.else, gravity_override="0, 0,-300")
   n5 = Branch(execute=@n1.then_1, Condition=Get bIsOstrichVehicleDriver() -> (bIsOstrichVehicleDriver:bool))
   Set in vehicle rigid alpha(execute=[@n6.then, @n7.then, @n5.then], in vehicle rigid alpha:double=0) -> (Output_Get:double)
   > Set in vehicle scale alpha(in vehicle scale alpha:double=1) -> (Output_Get:double)
   n6 = Branch(execute=@n5.else, Condition=Get bIsFerretVehicleDriver() -> (bIsFerretVehicleDriver:bool))
   n7 = Branch(execute=@n6.else, Condition=Get bIsOctopusVehicleDriver() -> (bIsOctopusVehicleDriver:bool))
   Set in vehicle rigid alpha(execute=@n7.else, in vehicle rigid alpha=1)
   > Set in vehicle scale alpha(in vehicle scale alpha=0)
   n8 = Get Skel_Mesh()
   Set thigh_rot_r(
    execute=@n1.then_2
    thigh_rot_r=GetAlphaFromReferenceTransformDelta(SkeletalMeshComponent=@n8, BoneName=thigh_r, BaseName=pelvis, InputRange=(X=0,Y=-90), TwistAxis="0,0,1")
   )
   > Set thigh_rot_l(thigh_rot_l=GetAlphaFromReferenceTransformDelta(SkeletalMeshComponent=@n8, BoneName=thigh_l, BaseName=pelvis, InputRange=(X=-20,Y=-101), TwistAxis="0,0,1"))
   Set isSkydiving(execute=@n1.then_3, isSkydiving:double="To Float (Boolean)"(InBool=Get bIsSkyDiving())) -> (Output_Get:double)
   n9 = Get bIsSliding()
   n10 = Get bIsCrouching()
   > Set isCrouchSliding(isCrouchSliding:double="To Float (Boolean)"(InBool=AND Boolean(A=@n9, B=@n10))) -> (Output_Get:double)
   > Set FeOrEmote(FeOrEmote:bool=OR Boolean(A=Get bIsPlayingEmote(), B=Get bIsInFrontEnd())) -> (Output_Get:bool)
   n11 = Sequence!(execute=@n1.then_4) -> (then_0, then_1, then_2)
   n12 = BreakRotator(InRot=BreakTransform(InTransform=GetDeltaTransformFromRefPose(self=@n8, BoneName=thigh_r, BaseName=pelvis)).Rotation)
   n13 = "float < float"(A=@n12.Yaw, B=-12)
   n14 = "float > float"(A=@n12.Pitch, B=12)
   > n15 = Branch(Condition=AND Boolean(A=OR Boolean(A=@n14, B=@n13), B=Get FeOrEmote() -> (FeOrEmote:bool)))
   > Set LegRot_r(LegRot_r:double=1) -> (Output_Get:double)
   Set LegRot_r(execute=@n15.else, LegRot_r=0)
   n16 = BreakRotator(InRot=BreakTransform(InTransform=GetDeltaTransformFromRefPose(self=@n8, BoneName=thigh_l, BaseName=pelvis)).Rotation)
   n17 = "float > float"(A=@n16.Pitch, B=15)
   n18 = "float < float"(A=@n16.Yaw, B=-4)
   n19 = Branch(execute=@n11.then_1, Condition=AND Boolean(A=OR Boolean(A=@n17, B=@n18), B=Get FeOrEmote()))
   > Set LegRot_l(LegRot_l:double=1) -> (Output_Get:double)
   Set LegRot_l(execute=@n19.else, LegRot_l=0)
   n20 = BreakRotator(InRot=BreakTransform(InTransform=GetDeltaTransformFromRefPose(self=@n8, BoneName=thigh_r, BaseName=pelvis)).Rotation)
   n21 = Branch(execute=@n11.then_2, Condition=AND Boolean(A="float > float"(A=@n20.Yaw, B=12), B=Get FeOrEmote()))
   > Set LegBackRot_r(LegBackRot_r:double=1) -> (Output_Get:double)
   Set LegBackRot_r(execute=@n21.else, LegBackRot_r=0)
   Event BlueprintInitializeAnimation()
   > Set Skel_Mesh(Skel_Mesh=GetOwnerSkeletalMeshComponent())
  ]}
]
variables [
 FeOrEmote {type=boolean}
 gravity_override {type=Vector}
 in vehicle rigid alpha {type=number}
 in vehicle scale alpha {type=number}
 isCrouchSliding {type=number}
 isSkydiving {type=number}
 LegBackRot_r {type=number}
 LegRot_l {type=number}
 LegRot_r {type=number}
 Skel_Mesh {type=SkeletalMeshComponent}
 thigh_rot_l {type=number}
 thigh_rot_r {type=number}
]
