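# Physics asset for the face accessory mesh of the Tailor male medium player cosmetic body, giving physics shapes to a front hair bone.
# It has two capsule bodies, one on the head bone and a smaller offset one on dyn_hair_fr_01.
# A single constraint links the hair bone to the head, with limited swing on both axes and a tight twist limit.
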PhysicsAsset /BRCosmetics/Characters/Player/Male/Medium/Bodies/M_MED_Tailor/Meshes/Parts/M_MED_Tailor_FaceAcc_Physics
bodies {
 dyn_hair_fr_01 [{shapeType=Capsule, length=2.0191, radius=1.2723, center=<0 0 -1.3769>, rotation=<rot 90 0 0>}]
 head [{shapeType=Capsule, length=5, radius=5, rotation=<rot 90 0 0>}]
}
constraints [{bone1Name=dyn_hair_fr_01, bone2Name=head, swing1LimitDegrees=45, swing2LimitDegrees=45, twistLimitDegrees=5}]
